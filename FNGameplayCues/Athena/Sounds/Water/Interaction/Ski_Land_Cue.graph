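# Sound cue for the landing impact of a ski jump, with separate water and dirt variants.
# An integer switch on the OnLand parameter picks between random pools of four water or four dirt impacts, with water also serving as the fallback.
# A root modulator feeds the result out through a swimming tether attenuation and the water interaction sound class.
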SoundCue /FNGameplayCues/Athena/Sounds/Water/Interaction/Ski_Land_Cue
root=Modulator_0
nodes [
 {id=Modulator_0, children [Switch_0]}
 {id=Random_0, children [WavePlayer_4,WavePlayer_5,WavePlayer_6,WavePlayer_7], properties {weights [1,1,1,1]}}
 {id=Random_1, children [WavePlayer_8, WavePlayer_9, WavePlayer_10, WavePlayer_11], properties {weights [1,1,1,1]}}
 {id=Switch_0, children [Random_1,Random_1,Random_0], properties {intParameterName=OnLand}}
 {id=WavePlayer_10, properties {soundWaveAssetPtr=&/Ski_Water_Jump_Land_Impact_03}}
 {id=WavePlayer_11, properties {soundWaveAssetPtr=&/Ski_Water_Jump_Land_Impact_04}}
 {id=WavePlayer_4, properties {soundWaveAssetPtr=&/Ski_Dirt_Jump_Land_Impact_04}}
 {id=WavePlayer_5, properties {soundWaveAssetPtr=&/Ski_Dirt_Jump_Land_Impact_01}}
 {id=WavePlayer_6, properties {soundWaveAssetPtr=&/Ski_Dirt_Jump_Land_Impact_02}}
 {id=WavePlayer_7, properties {soundWaveAssetPtr=&/Ski_Dirt_Jump_Land_Impact_03}}
 {id=WavePlayer_8, properties {soundWaveAssetPtr=&/Ski_Water_Jump_Land_Impact_01}}
 {id=WavePlayer_9, properties {soundWaveAssetPtr=&/Ski_Water_Jump_Land_Impact_02}}
]
properties {
 attenuationSettings=/FNGameplayCues/Sounds/Attenuation/Fort_Player_Swimming_Tether_Athena_Att
 maxDistance=8000
 soundClassObject=/Game/Sounds/FortSoundClassesAndModes/Fort_Water_Interaction
 volumeMultiplier=1.25
}
