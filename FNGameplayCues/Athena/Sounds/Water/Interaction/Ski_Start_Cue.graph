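# Sound cue for the start of water tether skiing in Athena, likely played when a player begins being towed across water.
# A single wave player feeds the root with the Ski_Water_Tether_Start wave, with no randomization or modulation.
# It uses the swimming tether attenuation and the water interaction sound class, with boosted volume.
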SoundCue /FNGameplayCues/Athena/Sounds/Water/Interaction/Ski_Start_Cue
root=WavePlayer_4
nodes [{id=WavePlayer_4, properties {soundWaveAssetPtr=&/Ski_Water_Tether_Start}}]
properties {
 attenuationSettings=/FNGameplayCues/Sounds/Attenuation/Fort_Player_Swimming_Tether_Athena_Att
 maxDistance=8000
 soundClassObject=/Game/Sounds/FortSoundClassesAndModes/Fort_Water_Interaction
 volumeMultiplier=2
}
