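# Looping SoundCue for a player skiing on water, covering wake, dirt slide and wind sounds.
# A Surface parameter crossfades between the water wake mix, a dirt slide loop and a skydive wind loop.
# Within the water mix, Steering fades a carve loop in over the constant wake.
# Velocity scales the blended volume, and Steering further boosts it at the root, with swimming-tether attenuation applied.
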SoundCue /FNGameplayCues/Athena/Sounds/Water/Interaction/Ski_Wake_Loop_Cue
root=ModulatorContinuous_3
nodes [
 {id=Mixer_0, children [WavePlayer_0, ModulatorContinuous_0], properties {inputVolume [1,1.2]}}
 {
  id=ModulatorContinuous_0
  children [WavePlayer_3]
  properties {
   pitchModulationParams {default=0, maxInput=0.5, minOutput=1, parameterName=Steering}
   volumeModulationParams {default=0, maxInput=0.5, parameterName=Steering}
  }
 }
 {
  id=ModulatorContinuous_2
  children [ParamCrossFade_0]
  properties {pitchModulationParams {default=0, minOutput=1, parameterName=Velocity}, volumeModulationParams {default=0, parameterName=Velocity}}
 }
 {
  id=ModulatorContinuous_3
  children [ModulatorContinuous_2]
  properties {
   pitchModulationParams {default=0, maxInput=0.5, minOutput=1, parameterName=Steering}
   volumeModulationParams {default=0, maxInput=0.5, maxOutput=1.7, minOutput=1, parameterName=Steering}
  }
 }
 {
  id=ParamCrossFade_0
  children [Mixer_0,WavePlayer_2,WavePlayer_1]
  properties {
   paramName=Surface
   crossFadeInput [
    {fadeInDistanceEnd=0, fadeInDistanceStart=0, fadeOutDistanceEnd=0.3333, fadeOutDistanceStart=0, volume=1}
    {fadeInDistanceEnd=0.3333, fadeInDistanceStart=0, fadeOutDistanceEnd=0.6667, fadeOutDistanceStart=0.3333, volume=1}
    {fadeInDistanceEnd=0.6667, fadeInDistanceStart=0.3333, fadeOutDistanceEnd=100, fadeOutDistanceStart=100, volume=1}
   ]
  }
 }
 {id=WavePlayer_0, properties {bLooping=true, soundWaveAssetPtr=/Game/Athena/Sounds/Water/Interaction/Ski_Water_Player_Wake_Constant_Loop_01}}
 {id=WavePlayer_1, properties {bLooping=true, soundWaveAssetPtr=/Game/Athena/Sounds/Wind_Falling/AMB_Wind_Skydive_Horizontal_Loop_01}}
 {id=WavePlayer_2, properties {bLooping=true, soundWaveAssetPtr=&/Ski_Dirt_Player_Slide_Loop_01}}
 {
  id=WavePlayer_3
  properties {bLooping=true, soundWaveAssetPtr=/Game/Athena/Sounds/Water/Interaction/Ski_Water_Player_Wake_Carve_Constant_Loop_01}
 }
]
properties {
 attenuationSettings=/FNGameplayCues/Sounds/Attenuation/Fort_Player_Swimming_Tether_Athena_Att
 maxDistance=8000
 soundClassObject=/Game/Sounds/FortSoundClassesAndModes/Fort_Water_Interaction
 volumeMultiplier=1
}
