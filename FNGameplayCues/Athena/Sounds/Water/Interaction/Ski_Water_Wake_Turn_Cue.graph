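# Sound cue for the water wake heard during ski turns, audible only when the player is in water.
# An IsInWater switch stays silent by default and at 0, and at 1 plays a random pick of five equally weighted wake-turn waves through a pitch-varying modulator.
# It uses swimming-tether attenuation and the water interaction sound class, with per-owner concurrency that stops the oldest instance.
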SoundCue /FNGameplayCues/Athena/Sounds/Water/Interaction/Ski_Water_Wake_Turn_Cue
root=Switch_0
nodes [
 {id=Modulator_0, children [Random_0], properties {pitchMax=1.4, pitchMin=0.8}}
 {id=Random_0, children [WavePlayer_0,WavePlayer_1,WavePlayer_5,WavePlayer_6,WavePlayer_7], properties {weights [1,1,1,1,1]}}
 {id=Switch_0, children [null,null,Modulator_0], properties {intParameterName=IsInWater}}
 {id=WavePlayer_0, properties {soundWaveAssetPtr=&/Ski_Water_Player_Wake_Turn_01}}
 {id=WavePlayer_1, properties {soundWaveAssetPtr=&/Ski_Water_Player_Wake_Turn_02}}
 {id=WavePlayer_5, properties {soundWaveAssetPtr=&/Ski_Water_Player_Wake_Turn_04}}
 {id=WavePlayer_6, properties {soundWaveAssetPtr=&/Ski_Water_Player_Wake_Turn_05}}
 {id=WavePlayer_7, properties {soundWaveAssetPtr=&/Ski_Water_Player_Wake_Turn_06}}
]
properties {
 attenuationSettings=/FNGameplayCues/Sounds/Attenuation/Fort_Player_Swimming_Tether_Athena_Att
 bOverrideConcurrency=true
 maxDistance=8000
 soundClassObject=/Game/Sounds/FortSoundClassesAndModes/Fort_Water_Interaction
 volumeMultiplier=1.5
 concurrencyOverrides {bLimitToOwner=true, maxCount=2, resolutionRule=StopOldest, volumeScale=0.6476}
}
