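# Sound cue played for a player's drowning damage tick in Athena water interaction.
# A random node picks one of three equally weighted body hit-marker waves, reusing hit notification sounds rather than water-specific audio.
# The random node is excluded from branch culling, and playback routes through the water interaction sound class with swimming attenuation.
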SoundCue /FNGameplayCues/Athena/Sounds/Water/Interaction/Water_Player_Drown_DamageTick_Cue
root=Random_0
nodes [
 {id=Random_0, children [WavePlayer_2,WavePlayer_3,WavePlayer_4], properties {bShouldExcludeFromBranchCulling=true, weights [1,1,1]}}
 {id=WavePlayer_2, properties {soundWaveAssetPtr=/Game/Athena/Sounds/HitNotifications/HitMarker_Body_01}}
 {id=WavePlayer_3, properties {soundWaveAssetPtr=/Game/Athena/Sounds/HitNotifications/HitMarker_Body_02}}
 {id=WavePlayer_4, properties {soundWaveAssetPtr=/Game/Athena/Sounds/HitNotifications/HitMarker_Body_03}}
]
properties {
 attenuationSettings=/Game/Sounds/Attenuation/Fort_Player_Swimming_Athena_Att
 maxDistance=4500
 soundClassObject=/Game/Sounds/FortSoundClassesAndModes/Fort_Water_Interaction
 volumeMultiplier=1
}
